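# Sound cue for the Smackie creature's alert vocalization, played as an attenuated, tactical creature vocalization.
# A quality-level switch picks between two equally weighted random pools of three "Alerted" waves, with the higher levels using the Revised_Alerted folder.
# An enveloper ramps pitch upward over roughly half a second, then a modulator adds slight random upward pitch variation.
# Concurrency is governed by Athena footstep local and global sets rather than vocalization-specific ones.
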SoundCue /Irwin/Sounds/Smackie/Vocalizations/Smackie_Alert_Cue
root=Modulator_1
nodes [
 {
  id=Enveloper_0
  children [QualityLevel_0]
  properties {pitchCurve {editorCurveData {keys [{time=0}, <at 0.4164 1.0901>, <at 0.5694 1.2458>]}}, volumeCurve {editorCurveData {keys [<at 0 1>]}}}
 }
 {id=Modulator_1, children [Enveloper_0], properties {pitchMax=1.1, pitchMin=1}}
 {id=QualityLevel_0, children [Random_2,Random_2,Random_3,Random_3]}
 {id=Random_2, children [WavePlayer_3,WavePlayer_4,WavePlayer_5], properties {weights [1,1,1]}}
 {id=Random_3, children [WavePlayer_11, WavePlayer_12, WavePlayer_13], properties {weights [1,1,1]}}
 {id=WavePlayer_11, properties {soundWaveAssetPtr=/Irwin/Sounds/Smackie/Vocalizations/Revised_Alerted/Fauna_Smackie_Alerted_01}}
 {id=WavePlayer_12, properties {soundWaveAssetPtr=/Irwin/Sounds/Smackie/Vocalizations/Revised_Alerted/Fauna_Smackie_Alerted_02}}
 {id=WavePlayer_13, properties {soundWaveAssetPtr=/Irwin/Sounds/Smackie/Vocalizations/Revised_Alerted/Fauna_Smackie_Alerted_03}}
 {id=WavePlayer_3, properties {soundWaveAssetPtr=&Fauna_Smackie_Alerted_01}}
 {id=WavePlayer_4, properties {soundWaveAssetPtr=&Fauna_Smackie_Alerted_02}}
 {id=WavePlayer_5, properties {soundWaveAssetPtr=&Fauna_Smackie_Alerted_03}}
]
properties {
 attenuationSettings=&/Prey_Smackie_Vocalization_Att
 maxDistance=4500
 priority=10
 soundClassObject=&Fort_Creatures_Vocalization_Tactical
 concurrencySet [&Athena_Footstep_Local_Conc, &Athena_Footstep_Global_Conc]
}
