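# Sound cue for the death vocalization of the Smackie creature.
# A quality-level switch picks one of two mixers, each layering a random death call over a quieter random jump sound.
# The first two quality slots draw on five jump variations, the last two on three.
# Output is pitched up and attenuated, in the creature vocalization sound class, with footstep concurrency groups limiting playback.
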SoundCue /Irwin/Sounds/Smackie/Vocalizations/Smackie_Death_Cue
root=QualityLevel_0
nodes [
 {id=Mixer_0, children [Random_5,Random_2], properties {inputVolume [1,0.35]}}
 {id=Mixer_1, children [Random_6,Random_1], properties {inputVolume [1,0.35]}}
 {id=QualityLevel_0, children [Mixer_0,Mixer_0,Mixer_1,Mixer_1]}
 {id=Random_1, children [WavePlayer_0,WavePlayer_6,WavePlayer_7], properties {bShouldExcludeFromBranchCulling=true, weights [1,1,1]}}
 {id=Random_2, children [WavePlayer_19, WavePlayer_20, WavePlayer_21, WavePlayer_22, WavePlayer_23], properties {weights [1,1,1,1,1]}}
 {id=Random_5, children [WavePlayer_12, WavePlayer_13, WavePlayer_15], properties {weights [1,1,1]}}
 {id=Random_6, children [WavePlayer_16, WavePlayer_17, WavePlayer_18], properties {weights [1,1,1]}}
 {id=WavePlayer_0, properties {soundWaveAssetPtr=/Irwin/Sounds/Smackie/Vocalizations/Original/Smackie_Jump_SFX_01}}
 {id=WavePlayer_12, properties {soundWaveAssetPtr=/Irwin/Sounds/Smackie/Vocalizations/Revised_Alerted/Fauna_Smackie_Death_01}}
 {id=WavePlayer_13, properties {soundWaveAssetPtr=/Irwin/Sounds/Smackie/Vocalizations/Revised_Alerted/Fauna_Smackie_Death_02}}
 {id=WavePlayer_15, properties {soundWaveAssetPtr=/Irwin/Sounds/Smackie/Vocalizations/Revised_Alerted/Fauna_Smackie_Death_03}}
 {id=WavePlayer_16, properties {soundWaveAssetPtr=&Fauna_Smackie_Death_01}}
 {id=WavePlayer_17, properties {soundWaveAssetPtr=&Fauna_Smackie_Death_02}}
 {id=WavePlayer_18, properties {soundWaveAssetPtr=&Fauna_Smackie_Death_03}}
 {id=WavePlayer_19, properties {soundWaveAssetPtr=&Smackie_Jump_SFX_01}}
 {id=WavePlayer_20, properties {soundWaveAssetPtr=/Irwin/Sounds/Smackie/Vocalizations/Original/Smackie_Jump_SFX_02}}
 {id=WavePlayer_21, properties {soundWaveAssetPtr=/Irwin/Sounds/Smackie/Vocalizations/Original/Smackie_Jump_SFX_03}}
 {id=WavePlayer_22, properties {soundWaveAssetPtr=/Irwin/Sounds/Smackie/Vocalizations/Original/Smackie_Jump_SFX_04}}
 {id=WavePlayer_23, properties {soundWaveAssetPtr=/Irwin/Sounds/Smackie/Vocalizations/Original/Smackie_Jump_SFX_05}}
 {id=WavePlayer_6, properties {soundWaveAssetPtr=&Smackie_Jump_SFX_02}}
 {id=WavePlayer_7, properties {soundWaveAssetPtr=&Smackie_Jump_SFX_03}}
]
properties {
 attenuationSettings=&/Prey_Smackie_Vocalization_Att
 maxDistance=4500
 pitchMultiplier=1.2
 soundClassObject=/Game/Sounds/FortSoundClassesAndModes/Fort_Creatures_Vocalization_Normal
 volumeMultiplier=0.4
 concurrencySet [&Athena_Footstep_Local_Conc, &Athena_Footstep_Global_Conc]
}
