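# Sound cue for the Smackie creature's alerted hop, layering an alert vocalization over jump movement sounds.
# Randomized vocals get pitch and volume modulation plus slight random delay, then mix with quieter jump foley and jump SFX, each layer using its own attenuation.
# A quality switch gives the two lower levels full variation pools and the two higher levels smaller ones.
# Some random pools include empty slots for occasional silence; footstep concurrency limits playback.
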SoundCue /Irwin/Sounds/Smackie/Vocalizations/Smackie_Hop_Alert_Cue
root=QualityLevel_0
nodes [
 {id=Attenuation_0, children [Delay_0], properties {attenuationSettings=&/Prey_Smackie_Vocalization_Att}}
 {id=Attenuation_1, children [Mixer_1], properties {attenuationSettings=/Irwin/Sounds/Settings/Prey_Step_Att}}
 {id=Attenuation_2, children [Delay_1], properties {attenuationSettings=&Prey_Smackie_Vocalization_Att}}
 {id=Attenuation_3, children [Mixer_3], properties {attenuationSettings=&Prey_Step_Att}}
 {id=Delay_0, children [Modulator_0], properties {delayMax=0.08}}
 {id=Delay_1, children [Modulator_1], properties {delayMax=0.08}}
 {id=Mixer_0, children [Attenuation_0, Attenuation_1], properties {inputVolume [1,0.25]}}
 {id=Mixer_1, children [Random_6,Random_3], properties {inputVolume [0.5,0.75]}}
 {id=Mixer_2, children [Attenuation_2, Attenuation_3], properties {inputVolume [1,0.25]}}
 {id=Mixer_3, children [Random_2,Random_0], properties {inputVolume [0.5,0.75]}}
 {id=Modulator_0, children [Random_4], properties {pitchMin=0.99, volumeMin=0.5}}
 {id=Modulator_1, children [Random_7], properties {pitchMax=1.1, pitchMin=0.99, volumeMin=0.5}}
 {id=QualityLevel_0, children [Mixer_0,Mixer_0,Mixer_2,Mixer_2]}
 {id=Random_0, children [WavePlayer_0,WavePlayer_1,WavePlayer_2], properties {bShouldExcludeFromBranchCulling=true, weights [1,1,1]}}
 {id=Random_2, children [WavePlayer_10, WavePlayer_11, null], properties {bShouldExcludeFromBranchCulling=true, weights [1,1,1]}}
 {id=Random_3, children [WavePlayer_17, WavePlayer_18, WavePlayer_19, WavePlayer_20, WavePlayer_21], properties {weights [1,1,1,1,1]}}
 {id=Random_4, children [WavePlayer_4,WavePlayer_7,WavePlayer_8,null], properties {weights [1,1,1,1]}}
 {id=Random_6, children [WavePlayer_32, WavePlayer_33, WavePlayer_34, WavePlayer_35, WavePlayer_36], properties {weights [1,1,1,1,1]}}
 {id=Random_7, children [WavePlayer_9, WavePlayer_12], properties {weights [1,1]}}
 {id=WavePlayer_0, properties {soundWaveAssetPtr=/Irwin/Sounds/Smackie/Vocalizations/Original/Smackie_Jump_SFX_05}}
 {id=WavePlayer_1, properties {soundWaveAssetPtr=/Irwin/Sounds/Smackie/Vocalizations/Original/Smackie_Jump_SFX_01}}
 {id=WavePlayer_10, properties {soundWaveAssetPtr=/Irwin/Sounds/Smackie/Foley/Fauna_Smackie_Jump_SFX_02}}
 {id=WavePlayer_11, properties {soundWaveAssetPtr=/Irwin/Sounds/Smackie/Foley/Fauna_Smackie_Jump_SFX_03}}
 {id=WavePlayer_12, properties {soundWaveAssetPtr=/Irwin/Sounds/Smackie/Vocalizations/Revised_Alerted/Fauna_Smackie_Alerted_02}}
 {id=WavePlayer_17, properties {soundWaveAssetPtr=&Smackie_Jump_SFX_05}}
 {id=WavePlayer_18, properties {soundWaveAssetPtr=&Smackie_Jump_SFX_01}}
 {id=WavePlayer_19, properties {soundWaveAssetPtr=/Irwin/Sounds/Smackie/Vocalizations/Original/Smackie_Jump_SFX_02}}
 {id=WavePlayer_2, properties {soundWaveAssetPtr=&Smackie_Jump_SFX_02}}
 {id=WavePlayer_20, properties {soundWaveAssetPtr=/Irwin/Sounds/Smackie/Vocalizations/Original/Smackie_Jump_SFX_03}}
 {id=WavePlayer_21, properties {soundWaveAssetPtr=/Irwin/Sounds/Smackie/Vocalizations/Original/Smackie_Jump_SFX_04}}
 {id=WavePlayer_32, properties {soundWaveAssetPtr=/Irwin/Sounds/Smackie/Foley/Fauna_Smackie_Jump_SFX_01}}
 {id=WavePlayer_33, properties {soundWaveAssetPtr=&Fauna_Smackie_Jump_SFX_02}}
 {id=WavePlayer_34, properties {soundWaveAssetPtr=&Fauna_Smackie_Jump_SFX_03}}
 {id=WavePlayer_35, properties {soundWaveAssetPtr=/Irwin/Sounds/Smackie/Foley/Fauna_Smackie_Jump_SFX_04}}
 {id=WavePlayer_36, properties {soundWaveAssetPtr=/Irwin/Sounds/Smackie/Foley/Fauna_Smackie_Jump_SFX_05}}
 {id=WavePlayer_4, properties {soundWaveAssetPtr=/Irwin/Sounds/Smackie/Vocalizations/Revised_Alerted/Fauna_Smackie_Alerted_01}}
 {id=WavePlayer_7, properties {soundWaveAssetPtr=&Fauna_Smackie_Alerted_02}}
 {id=WavePlayer_8, properties {soundWaveAssetPtr=/Irwin/Sounds/Smackie/Vocalizations/Revised_Alerted/Fauna_Smackie_Alerted_03}}
 {id=WavePlayer_9, properties {soundWaveAssetPtr=&Fauna_Smackie_Alerted_01}}
]
properties {
 maxDistance=4500
 soundClassObject=&Fort_Creatures_Vocalization_Tactical
 volumeMultiplier=0.65
 concurrencySet [&Athena_Footstep_Local_Conc, &Athena_Footstep_Global_Conc]
}
