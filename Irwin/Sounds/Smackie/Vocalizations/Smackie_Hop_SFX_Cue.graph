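# Sound cue for the hop sound of the Smackie creature, playing jump foley waves at low volume on the creature vocalization sound class.
# A quality-level switch sends its first two levels to a random pick of five variations and its last two to a pick of three.
# A modulator adds slight pitch variation.
# Prey step attenuation, footstep concurrency and a one-per-owner stop-oldest rule govern playback.
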SoundCue /Irwin/Sounds/Smackie/Vocalizations/Smackie_Hop_SFX_Cue
root=Modulator_0
nodes [
 {id=Modulator_0, children [QualityLevel_0], properties {pitchMax=1.02, pitchMin=0.98}}
 {id=QualityLevel_0, children [Random_1,Random_1,Random_0,Random_0]}
 {id=Random_0, children [WavePlayer_1,WavePlayer_2,WavePlayer_3], properties {bShouldExcludeFromBranchCulling=true, weights [1,1,1]}}
 {id=Random_1, children [WavePlayer_0,WavePlayer_6,WavePlayer_7,WavePlayer_8,WavePlayer_9], properties {weights [1,1,1,1,1]}}
 {id=WavePlayer_0, properties {soundWaveAssetPtr=/Irwin/Sounds/Smackie/Foley/Fauna_Smackie_Jump_SFX_01}}
 {id=WavePlayer_1, properties {soundWaveAssetPtr=&Fauna_Smackie_Jump_SFX_01}}
 {id=WavePlayer_2, properties {soundWaveAssetPtr=/Irwin/Sounds/Smackie/Foley/Fauna_Smackie_Jump_SFX_02}}
 {id=WavePlayer_3, properties {soundWaveAssetPtr=/Irwin/Sounds/Smackie/Foley/Fauna_Smackie_Jump_SFX_03}}
 {id=WavePlayer_6, properties {soundWaveAssetPtr=&Fauna_Smackie_Jump_SFX_02}}
 {id=WavePlayer_7, properties {soundWaveAssetPtr=&Fauna_Smackie_Jump_SFX_03}}
 {id=WavePlayer_8, properties {soundWaveAssetPtr=/Irwin/Sounds/Smackie/Foley/Fauna_Smackie_Jump_SFX_04}}
 {id=WavePlayer_9, properties {soundWaveAssetPtr=/Irwin/Sounds/Smackie/Foley/Fauna_Smackie_Jump_SFX_05}}
]
properties {
 attenuationSettings=/Irwin/Sounds/Settings/Prey_Step_Att
 maxDistance=3200
 soundClassObject=/Game/Sounds/FortSoundClassesAndModes/Fort_Creatures_Vocalization_Normal
 volumeMultiplier=0.025
 concurrencySet [
  &Athena_Footstep_Local_Conc
  &Athena_Footstep_Global_Conc
  /Game/Sounds/Concurrency_Rules/OneMaximum_Owner_StopOldest
 ]
}
