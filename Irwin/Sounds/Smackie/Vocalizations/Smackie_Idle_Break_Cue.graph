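# Sound cue for Smackie's idle-break vocalization, playing a random tongue sound at reduced volume and slightly raised pitch.
# A modulator adds subtle pitch variation over a quality-level switch, whose first two levels pick from three waves and last two from two.
# It routes to the creature foley sound class with prey vocalization attenuation, and borrows local and global footstep concurrency sets to limit simultaneous instances.
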SoundCue /Irwin/Sounds/Smackie/Vocalizations/Smackie_Idle_Break_Cue
root=Modulator_0
nodes [
 {id=Modulator_0, children [QualityLevel_0], properties {pitchMax=1.021, pitchMin=0.98}}
 {id=QualityLevel_0, children [Random_3,Random_3,Random_4,Random_4]}
 {id=Random_3, children [WavePlayer_10, WavePlayer_11, WavePlayer_9], properties {weights [1,1,1]}}
 {id=Random_4, children [WavePlayer_13, WavePlayer_14], properties {weights [1,1]}}
 {id=WavePlayer_10, properties {soundWaveAssetPtr=/Irwin/Sounds/Smackie/Vocalizations/Original/Smackie_Tongue_10}}
 {id=WavePlayer_11, properties {soundWaveAssetPtr=/Irwin/Sounds/Smackie/Vocalizations/Original/Smackie_Tongue_09}}
 {id=WavePlayer_13, properties {soundWaveAssetPtr=&Smackie_Tongue_10}}
 {id=WavePlayer_14, properties {soundWaveAssetPtr=&Smackie_Tongue_09}}
 {id=WavePlayer_9, properties {soundWaveAssetPtr=/Irwin/Sounds/Smackie/Vocalizations/Original/Smackie_Tongue_08}}
]
properties {
 attenuationSettings=&/Prey_Smackie_Vocalization_Att
 maxDistance=4500
 pitchMultiplier=1.1
 soundClassObject=/Game/Sounds/FortSoundClassesAndModes/Fort_Creatures_Foley
 volumeMultiplier=0.25
 concurrencySet [&Athena_Footstep_Local_Conc, &Athena_Footstep_Global_Conc]
}
