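# Sound cue for Smackie creature idle vocalizations, played at half volume with prey vocalization attenuation on the creature vocalization sound class.
# A quality-level node sends the first two levels to a seven-wave random pool and the last two to a three-wave pool.
# A modulator at the root varies pitch slightly and volume widely on each play.
# Concurrency uses Athena local and global footstep sets rather than vocalization-specific ones.
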SoundCue /Irwin/Sounds/Smackie/Vocalizations/Smackie_Idle_Cue
root=Modulator_0
nodes [
 {id=Modulator_0, children [QualityLevel_0], properties {pitchMax=1.02, pitchMin=0.97, volumeMax=1, volumeMin=0.25}}
 {id=QualityLevel_0, children [Random_0,Random_0,Random_2,Random_2]}
 {
  id=Random_0
  children [WavePlayer_2,WavePlayer_3,WavePlayer_4,WavePlayer_5,WavePlayer_6,WavePlayer_7,WavePlayer_8]
  properties {weights [1,1,1,1,1,1,1]}
 }
 {id=Random_2, children [WavePlayer_0,WavePlayer_1,WavePlayer_9], properties {weights [1,1,1]}}
 {id=WavePlayer_0, properties {soundWaveAssetPtr=/Irwin/Sounds/Smackie/Vocalizations/Revised_Alerted/Fauna_Smackie_Idle_02}}
 {id=WavePlayer_1, properties {soundWaveAssetPtr=/Irwin/Sounds/Smackie/Vocalizations/Revised_Alerted/Fauna_Smackie_Idle_03}}
 {id=WavePlayer_2, properties {soundWaveAssetPtr=&Fauna_Smackie_Idle_02}}
 {id=WavePlayer_3, properties {soundWaveAssetPtr=&Fauna_Smackie_Idle_03}}
 {id=WavePlayer_4, properties {soundWaveAssetPtr=/Irwin/Sounds/Smackie/Vocalizations/Revised_Alerted/Fauna_Smackie_Idle_04}}
 {id=WavePlayer_5, properties {soundWaveAssetPtr=/Irwin/Sounds/Smackie/Vocalizations/Revised_Alerted/Fauna_Smackie_Idle_05}}
 {id=WavePlayer_6, properties {soundWaveAssetPtr=/Irwin/Sounds/Smackie/Vocalizations/Revised_Alerted/Fauna_Smackie_Idle_06}}
 {id=WavePlayer_7, properties {soundWaveAssetPtr=/Irwin/Sounds/Smackie/Vocalizations/Revised_Alerted/Fauna_Smackie_Idle_07}}
 {id=WavePlayer_8, properties {soundWaveAssetPtr=/Irwin/Sounds/Smackie/Vocalizations/Revised_Alerted/Fauna_Smackie_Idle_08}}
 {id=WavePlayer_9, properties {soundWaveAssetPtr=&Fauna_Smackie_Idle_04}}
]
properties {
 attenuationSettings=&/Prey_Smackie_Vocalization_Att
 maxDistance=4500
 soundClassObject=/Game/Sounds/FortSoundClassesAndModes/Fort_Creatures_Vocalization_Normal
 volumeMultiplier=0.5
 concurrencySet [&Athena_Footstep_Local_Conc, &Athena_Footstep_Global_Conc]
}
