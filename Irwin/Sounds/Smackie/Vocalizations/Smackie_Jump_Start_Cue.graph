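# Sound cue for the Smackie creature's jump start, using prey vocalization attenuation and footstep concurrency.
# A quality-level node selects between two mixes, each layering random vocal sweeteners with pitched-up, randomized foley under a brief volume fade-in.
# The first two levels use larger wave pools, pitch the sweetener down and include weighted empty slots that often silence it; the last two use smaller pools.
# An outer modulator varies overall pitch and volume.
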SoundCue /Irwin/Sounds/Smackie/Vocalizations/Smackie_Jump_Start_Cue
root=Modulator_0
nodes [
 {
  id=Enveloper_0
  children [Modulator_1]
  properties {pitchCurve {editorCurveData {keys [{time=0}]}}, volumeCurve {editorCurveData {keys [<at 0 0.0205>,<at 0 0.005>,<at 0.0529 1.1186>]}}}
 }
 {
  id=Enveloper_1
  children [Modulator_2]
  properties {pitchCurve {editorCurveData {keys [{time=0}]}}, volumeCurve {editorCurveData {keys [<at 0 0.0205>,<at 0 0.005>,<at 0.0529 1.1186>]}}}
 }
 {id=Mixer_0, children [Modulator_3,Enveloper_0], properties {inputVolume [0.8,0.7]}}
 {id=Mixer_1, children [Random_2,Enveloper_1], properties {inputVolume [1,1]}}
 {id=Modulator_0, children [QualityLevel_0], properties {pitchMax=0.95, pitchMin=0.9, volumeMin=0.5}}
 {id=Modulator_1, children [Random_1], properties {pitchMax=1.5, pitchMin=1.4}}
 {id=Modulator_2, children [Random_3], properties {pitchMax=1.3, pitchMin=1.2}}
 {id=Modulator_3, children [Random_0], properties {pitchMax=0.9, pitchMin=0.85}}
 {id=QualityLevel_0, children [Mixer_0,Mixer_0,Mixer_1,Mixer_1]}
 {
  id=Random_0
  children [WavePlayer_0,WavePlayer_1,WavePlayer_2,WavePlayer_4,null,null,null,null]
  properties {weights [0.2,0.2,0.2,0.2,0.2,1,1,1]}
 }
 {
  id=Random_1
  children [WavePlayer_3, WavePlayer_9, WavePlayer_10, WavePlayer_11, WavePlayer_12, WavePlayer_13]
  properties {weights [0.2,0.2,0.2,0.2,1,1]}
 }
 {id=Random_2, children [WavePlayer_6,WavePlayer_7,WavePlayer_8], properties {weights [1,1,1]}}
 {id=Random_3, children [WavePlayer_14, WavePlayer_15, WavePlayer_16], properties {weights [0.2,0.2,0.2]}}
 {id=WavePlayer_0, properties {soundWaveAssetPtr=/Irwin/Sounds/Smackie/Vocalizations/Revised_Alerted/Fauna_Smackie_Jump_Sweetner_01}}
 {id=WavePlayer_1, properties {soundWaveAssetPtr=/Irwin/Sounds/Smackie/Vocalizations/Revised_Alerted/Fauna_Smackie_Jump_Sweetner_02}}
 {id=WavePlayer_10, properties {soundWaveAssetPtr=/Irwin/Sounds/Smackie/Foley/Fauna_Smackie_Jump_SFX_04}}
 {id=WavePlayer_11, properties {soundWaveAssetPtr=/Irwin/Sounds/Smackie/Foley/Fauna_Smackie_Jump_SFX_05}}
 {id=WavePlayer_12, properties {soundWaveAssetPtr=/Irwin/Sounds/Smackie/Foley/Fauna_Smackie_Jump_SFX_06}}
 {id=WavePlayer_13, properties {soundWaveAssetPtr=/Irwin/Sounds/Smackie/Foley/Fauna_Smackie_Jump_SFX_07}}
 {id=WavePlayer_14, properties {soundWaveAssetPtr=/Irwin/Sounds/Smackie/Foley/Fauna_Smackie_Jump_SFX_01}}
 {id=WavePlayer_15, properties {soundWaveAssetPtr=/Irwin/Sounds/Smackie/Foley/Fauna_Smackie_Jump_SFX_03}}
 {id=WavePlayer_16, properties {soundWaveAssetPtr=&Fauna_Smackie_Jump_SFX_04}}
 {id=WavePlayer_2, properties {soundWaveAssetPtr=/Irwin/Sounds/Smackie/Vocalizations/Revised_Alerted/Fauna_Smackie_Jump_Sweetner_03}}
 {id=WavePlayer_3, properties {soundWaveAssetPtr=&Fauna_Smackie_Jump_SFX_01}}
 {id=WavePlayer_4, properties {soundWaveAssetPtr=/Irwin/Sounds/Smackie/Vocalizations/Revised_Alerted/Fauna_Smackie_Jump_Sweetner_05}}
 {id=WavePlayer_6, properties {soundWaveAssetPtr=&Fauna_Smackie_Jump_Sweetner_01}}
 {id=WavePlayer_7, properties {soundWaveAssetPtr=&Fauna_Smackie_Jump_Sweetner_02}}
 {id=WavePlayer_8, properties {soundWaveAssetPtr=&Fauna_Smackie_Jump_Sweetner_03}}
 {id=WavePlayer_9, properties {soundWaveAssetPtr=&Fauna_Smackie_Jump_SFX_03}}
]
properties {
 attenuationSettings=/Irwin/Sounds/Settings/Prey_Vocalization_Att
 maxDistance=6350
 priority=0.9
 soundClassObject=/Game/Sounds/FortSoundClassesAndModes/Fort_Creatures_Vocalization_Normal
 volumeMultiplier=0.3
 concurrencySet [&Athena_Footstep_Local_Conc, &Athena_Footstep_Global_Conc]
}
