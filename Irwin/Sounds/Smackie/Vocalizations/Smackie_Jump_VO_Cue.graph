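# Sound cue for the jump vocalization of Smackie, a fauna creature, using prey vocalization attenuation and the creature vocalization sound class.
# A quality-level node sends its first two slots to an equally weighted random pick of six jump waves and its last two to a pick of three.
# A modulator then varies volume and slightly lowers pitch at random.
# Concurrency uses the Athena local and global footstep sets.
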SoundCue /Irwin/Sounds/Smackie/Vocalizations/Smackie_Jump_VO_Cue
root=Modulator_0
nodes [
 {id=Modulator_0, children [QualityLevel_0], properties {pitchMax=1, pitchMin=0.98, volumeMin=0.5}}
 {id=QualityLevel_0, children [Random_0,Random_0,Random_2,Random_2]}
 {id=Random_0, children [WavePlayer_0,WavePlayer_1,WavePlayer_2,WavePlayer_3,WavePlayer_4,WavePlayer_5], properties {weights [1,1,1,1,1,1]}}
 {id=Random_2, children [WavePlayer_6,WavePlayer_7,WavePlayer_8], properties {weights [1,1,1]}}
 {id=WavePlayer_0, properties {soundWaveAssetPtr=/Irwin/Sounds/Smackie/Vocalizations/Revised_Alerted/Fauna_Smackie_Jump_01}}
 {id=WavePlayer_1, properties {soundWaveAssetPtr=/Irwin/Sounds/Smackie/Vocalizations/Revised_Alerted/Fauna_Smackie_Jump_02}}
 {id=WavePlayer_2, properties {soundWaveAssetPtr=/Irwin/Sounds/Smackie/Vocalizations/Revised_Alerted/Fauna_Smackie_Jump_03}}
 {id=WavePlayer_3, properties {soundWaveAssetPtr=/Irwin/Sounds/Smackie/Vocalizations/Revised_Alerted/Fauna_Smackie_Jump_04}}
 {id=WavePlayer_4, properties {soundWaveAssetPtr=/Irwin/Sounds/Smackie/Vocalizations/Revised_Alerted/Fauna_Smackie_Jump_05}}
 {id=WavePlayer_5, properties {soundWaveAssetPtr=/Irwin/Sounds/Smackie/Vocalizations/Revised_Alerted/Fauna_Smackie_Jump_06}}
 {id=WavePlayer_6, properties {soundWaveAssetPtr=&Fauna_Smackie_Jump_01}}
 {id=WavePlayer_7, properties {soundWaveAssetPtr=&Fauna_Smackie_Jump_02}}
 {id=WavePlayer_8, properties {soundWaveAssetPtr=&Fauna_Smackie_Jump_03}}
]
properties {
 attenuationSettings=/Irwin/Sounds/Settings/Prey_Vocalization_Att
 maxDistance=6350
 priority=0.9
 soundClassObject=/Game/Sounds/FortSoundClassesAndModes/Fort_Creatures_Vocalization_Normal
 volumeMultiplier=0.45
 concurrencySet [&Athena_Footstep_Local_Conc, &Athena_Footstep_Global_Conc]
}
