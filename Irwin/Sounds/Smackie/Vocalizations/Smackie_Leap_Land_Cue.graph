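# Sound cue for the Smackie creature's leap landing, built from its jump foley waves despite sitting in the Vocalizations folder.
# A Random node picks among seven equally weighted Jump SFX variants, and a Modulator above it applies a slight random downward pitch shift.
# The cue plays at heavily reduced volume through prey footstep attenuation and the tactical creature vocalization sound class.
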SoundCue /Irwin/Sounds/Smackie/Vocalizations/Smackie_Leap_Land_Cue
root=Modulator_0
nodes [
 {id=Modulator_0, children [Random_0], properties {pitchMax=1, pitchMin=0.9}}
 {
  id=Random_0
  children [WavePlayer_1,WavePlayer_2,WavePlayer_3,WavePlayer_4,WavePlayer_5,WavePlayer_6,WavePlayer_7]
  properties {weights [1,1,1,1,1,1,1]}
 }
 {id=WavePlayer_1, properties {soundWaveAssetPtr=/Irwin/Sounds/Smackie/Foley/Fauna_Smackie_Jump_SFX_04}}
 {id=WavePlayer_2, properties {soundWaveAssetPtr=/Irwin/Sounds/Smackie/Foley/Fauna_Smackie_Jump_SFX_07}}
 {id=WavePlayer_3, properties {soundWaveAssetPtr=/Irwin/Sounds/Smackie/Foley/Fauna_Smackie_Jump_SFX_01}}
 {id=WavePlayer_4, properties {soundWaveAssetPtr=/Irwin/Sounds/Smackie/Foley/Fauna_Smackie_Jump_SFX_02}}
 {id=WavePlayer_5, properties {soundWaveAssetPtr=/Irwin/Sounds/Smackie/Foley/Fauna_Smackie_Jump_SFX_03}}
 {id=WavePlayer_6, properties {soundWaveAssetPtr=/Irwin/Sounds/Smackie/Foley/Fauna_Smackie_Jump_SFX_05}}
 {id=WavePlayer_7, properties {soundWaveAssetPtr=/Irwin/Sounds/Smackie/Foley/Fauna_Smackie_Jump_SFX_06}}
]
properties {
 attenuationSettings=/Irwin/Sounds/Settings/Prey_Step_Att
 maxDistance=3200
 soundClassObject=&Fort_Creatures_Vocalization_Tactical
 volumeMultiplier=0.1
}
